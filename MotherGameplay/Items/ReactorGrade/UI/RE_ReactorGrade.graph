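# Reticle extension widget for the Reactor Grade weapon, showing an animated aim-down-sights reticle with charge feedback.
# Targeting changes play the reticle and depth-shape intros forward or in reverse, while a fire event plays a ring burst and zeroes charge.
# While in ADS, charge events set FillPercent on dynamic ring and pip materials, and minimum charge triggers an arrow burst.
# Construct snaps animations to their start with fast reverse playback.
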WidgetBlueprint /MotherGameplay/Items/ReactorGrade/UI/RE_ReactorGrade
animations [
 ADS {
  binding_count=1
  length_seconds=0.2
  object=&self:ADS
  sequence {bindings [Overlay_19 {
     id=#1
     possessed_class=Overlay
     tracks [{
       type=MovieSceneFloatTrack
       display_name=Render Opacity
       sections [{type=MovieSceneFloatSection, channels [None {type=MovieSceneScriptingFloatChannel, keys [[0,0],[4,1]]}]}]
      }]
    }]}
 }
 ResetToBaseState {
  binding_count=1
  length_seconds=0.2
  object=&self:ResetToBaseState
  sequence {bindings [Overlay_19 {
     id=#1
     possessed_class=Overlay
     tracks [{
       type=MovieSceneFloatTrack
       display_name=Render Opacity
       sections [{type=MovieSceneFloatSection, end_frame=4, start_frame=0, channels [None {type=MovieSceneScriptingFloatChannel, const [0,1]}]}]
      }]
    }]}
 }
 ReticleIntro {
  binding_count=12
  length_seconds=0.6
  object=&self:ReticleIntro
  sequence {
   bindings [
    VerticalLine {
     id=#2
     possessed_class=Image
     tracks [{
       type=MovieScene2DTransformTrack
       sections [{type=MovieScene2DTransformSection, channels [Scale.Y {type=MovieSceneScriptingFloatChannel, keys [[0,0],[24,0],[26,1.3],[36,1]]}]}]
      }]
    }
    TimerPips_Base {
     id=#3
     possessed_class=Image
     tracks [
      {
       type=MovieSceneWidgetMaterialTrack
       sections [{
         type=MovieSceneParameterSection
         channels [
          RingWidth {type=MovieSceneScriptingFloatChannel, keys [[18.003,0.75],[24.006,3.5]]}
          Radius {type=MovieSceneScriptingFloatChannel, keys [[18.003,0.4],[24.006,0.36]]}
         ]
        }]
      }
      {
       type=MovieSceneFloatTrack
       display_name=Render Opacity
       sections [{type=MovieSceneFloatSection, channels [None {type=MovieSceneScriptingFloatChannel, keys [[0,0],[6,0],[9,1]]}]}]
      }
     ]
    }
    Charge_Progress {
     id=#4
     possessed_class=Image
     tracks [
      {
       type=MovieSceneFloatTrack
       display_name=Render Opacity
       sections [{type=MovieSceneFloatSection, channels [None {type=MovieSceneScriptingFloatChannel, keys [[0,0],[18,0],[30,1]]}]}]
      }
      {
       type=MovieSceneWidgetMaterialTrack
       sections [{type=MovieSceneParameterSection, channels [Overall_Alpha {type=MovieSceneScriptingFloatChannel, const [0.175,1]}]}]
      }
     ]
    }
    Databit_Left {
     id=#5
     possessed_class=Image
     tracks [{
       type=MovieSceneWidgetMaterialTrack
       sections [{
         type=MovieSceneParameterSection
         channels [
          ToBlur {type=MovieSceneScriptingFloatChannel, keys [[0,0],[30,1]]}
          GlowAmount {type=MovieSceneScriptingFloatChannel, keys [[0,1],[30,0]]}
          FinalAlpha {type=MovieSceneScriptingFloatChannel, keys [[0,0],[6,1],[12,1]]}
         ]
        }]
      }]
    }
    CenterRing {
     id=#6
     possessed_class=Image
     tracks [
      {
       type=MovieSceneWidgetMaterialTrack
       sections [{
         type=MovieSceneParameterSection
         channels [
          Radius {type=MovieSceneScriptingFloatChannel, thinned_from=5, keys [[0,0.75],[15,0.75],[16,0.5],[19,0.5]]}
          Overall_Alpha {type=MovieSceneScriptingFloatChannel, keys [[0,0],[3,0.5],[6,0.5],[9,0.05],[12,0.25],[15,0.25],[16,0.05]]}
          Density_Inner {type=MovieSceneScriptingFloatChannel, keys [[0,48],[15,48],[16,16]]}
         ]
        }]
      }
      {
       type=MovieScene2DTransformTrack
       sections [{
         type=MovieScene2DTransformSection
         channels [
          Scale.X {type=MovieSceneScriptingFloatChannel, keys [[0,1.5],[12,1.5],[15,1]]}
          Scale.Y {type=MovieSceneScriptingFloatChannel, keys [[0,1.5],[12,1.5],[15,1]]}
         ]
        }]
      }
     ]
    }
    Arrow {
     id=#7
     possessed_class=Image
     tracks [
      {
       type=MovieScene2DTransformTrack
       sections [{
         type=MovieScene2DTransformSection
         channels [
          Translation.X {type=MovieSceneScriptingFloatChannel}
          Translation.Y {type=MovieSceneScriptingFloatChannel, keys [[24,-80],[26.001,-64],[36,-70]]}
          Scale.X {type=MovieSceneScriptingFloatChannel}
          Scale.Y {type=MovieSceneScriptingFloatChannel}
         ]
        }]
      }
      {
       type=MovieSceneFloatTrack
       display_name=Render Opacity
       sections [{type=MovieSceneFloatSection, channels [None {type=MovieSceneScriptingFloatChannel, keys [[0,0],[24,0],[26,1]]}]}]
      }
     ]
    }
    Charge_Outer_Ring {
     id=#8
     possessed_class=Image
     tracks [{
       type=MovieSceneWidgetMaterialTrack
       sections [{
         type=MovieSceneParameterSection
         channels [
          FillPercent {type=MovieSceneScriptingFloatChannel, keys [[12,0],[24.007,100]]}
          Overall_Alpha {type=MovieSceneScriptingFloatChannel, keys [[0,1],[12,0.75],[36,0.5]]}
         ]
        }]
      }]
    }
    Databit_Right {
     id=#9
     possessed_class=Image
     tracks [{
       type=MovieSceneWidgetMaterialTrack
       sections [{
         type=MovieSceneParameterSection
         channels [
          ToBlur {type=MovieSceneScriptingFloatChannel, keys [[0,0],[30,1]]}
          GlowAmount {type=MovieSceneScriptingFloatChannel, keys [[0,1],[30,0]]}
          FinalAlpha {type=MovieSceneScriptingFloatChannel, keys [[0,0],[6,1],[12,1]]}
         ]
        }]
      }]
    }
    Charge_Progress_Backing {
     id=#10
     possessed_class=Image
     tracks [{
       type=MovieSceneWidgetMaterialTrack
       sections [{type=MovieSceneParameterSection, channels [RingWidth {type=MovieSceneScriptingFloatChannel, keys [[20.997,-2],[33,4]]}]}]
      }]
    }
    LargeBlueRing {
     id=#11
     possessed_class=Image
     tracks [
      {
       type=MovieSceneWidgetMaterialTrack
       sections [{type=MovieSceneParameterSection, channels [RingWidth {type=MovieSceneScriptingFloatChannel, keys [[9,4],[15,8]]}]}]
      }
      {
       type=MovieSceneFloatTrack
       display_name=Render Opacity
       sections [{type=MovieSceneFloatSection, channels [None {type=MovieSceneScriptingFloatChannel, keys [[6,0],[9,1]]}]}]
      }
      {
       type=MovieScene2DTransformTrack
       sections [{
         type=MovieScene2DTransformSection
         channels [
          Scale.X {type=MovieSceneScriptingFloatChannel, keys [[6,0.9],[12.001,1]]}
          Scale.Y {type=MovieSceneScriptingFloatChannel, keys [[6,0.9],[12.001,1]]}
         ]
        }]
      }
     ]
    }
    EncompassingRing_Shadow {
     id=#12
     possessed_class=Image
     tracks [{
       type=MovieSceneFloatTrack
       display_name=Render Opacity
       sections [{type=MovieSceneFloatSection, channels [None {type=MovieSceneScriptingFloatChannel, keys [[0,0],[18,1]]}]}]
      }]
    }
    EncompassingRing {
     id=#13
     possessed_class=Image
     tracks [{
       type=MovieSceneFloatTrack
       display_name=Render Opacity
       sections [{type=MovieSceneFloatSection, channels [None {type=MovieSceneScriptingFloatChannel, keys [[9,0],[18,1]]}]}]
      }]
    }
   ]
   display_rate {numerator=60}
  }
 }
 Fire {
  binding_count=1
  length_seconds=0.5
  object=&self:Fire
  sequence {bindings [LargeBlueRing {
     id=#14
     display_name=LargeBlueRing_Burst
     possessed_class=Image
     tracks [
      {
       type=MovieSceneWidgetMaterialTrack
       sections [{
         type=MovieSceneParameterSection
         channels [
          Overall_Alpha {type=MovieSceneScriptingFloatChannel, keys [[0,0.0125],[2,0.5],[10,0]]}
          RingWidth {type=MovieSceneScriptingFloatChannel, keys [[0,8],[10,-8]]}
         ]
        }]
      }
      {
       type=MovieScene2DTransformTrack
       sections [{
         type=MovieScene2DTransformSection
         channels [
          Translation.X {type=MovieSceneScriptingFloatChannel, const [0,1]}
          Translation.Y {type=MovieSceneScriptingFloatChannel, const [0,1]}
          Angle {type=MovieSceneScriptingFloatChannel, const [0,1]}
          Scale.X {type=MovieSceneScriptingFloatChannel, keys [[0,1],[10,2]]}
          Scale.Y {type=MovieSceneScriptingFloatChannel, keys [[0,1],[10,2]]}
         ]
        }]
      }
      {
       type=MovieSceneFloatTrack
       display_name=Render Opacity
       sections [{type=MovieSceneFloatSection, end_frame=10, start_frame=0, channels [None {type=MovieSceneScriptingFloatChannel, keys [[0,0],[1,1],[10,0]]}]}]
      }
     ]
    }]}
 }
 DepthShapesOutro {
  binding_count=8
  length_seconds=0.2
  object=&self:DepthShapesOutro
  sequence {
   bindings [
    Sights_Container_Left {
     id=#15
     possessed_class=Image
     tracks [
      {
       type=MovieScene2DTransformTrack
       sections [{
         type=MovieScene2DTransformSection
         channels [
          Translation.X {type=MovieSceneScriptingFloatChannel, keys [[0,0],[12,-48]]}
          Scale.X {type=MovieSceneScriptingFloatChannel, keys [[0,1],[12,1.125]]}
          Scale.Y {type=MovieSceneScriptingFloatChannel, keys [[0,1],[12,1.125]]}
         ]
        }]
      }
      {
       type=MovieSceneWidgetMaterialTrack
       sections [{
         type=MovieSceneParameterSection
         channels [
          FinalAlpha {type=MovieSceneScriptingFloatChannel, keys [[0,0.005],[12,0]]}
          Angle_1_Softness {type=MovieSceneScriptingFloatChannel, keys [[0,32],[12,16]]}
          Bottom_Cut_Softness {type=MovieSceneScriptingFloatChannel, keys [[0,40],[12,20]]}
          Bottom_Softness {type=MovieSceneScriptingFloatChannel, keys [[0,48],[12,24]]}
          Top_Cut_1_Softness {type=MovieSceneScriptingFloatChannel, keys [[0,40],[12,20]]}
          Top_Cut_2_Softness {type=MovieSceneScriptingFloatChannel, keys [[0,40],[12,20]]}
          Top_Softness {type=MovieSceneScriptingFloatChannel, keys [[0,48],[12,24]]}
          Angle_1_Width {type=MovieSceneScriptingFloatChannel, keys [[0,0.1405],[12,0.111]]}
          BottomOffset_Width {type=MovieSceneScriptingFloatChannel, keys [[0,-0.25],[12,-0.67]]}
          TopOffset_Width {type=MovieSceneScriptingFloatChannel, keys [[0,0.7205],[12,0.6905]]}
         ]
        }]
      }
     ]
    }
    Sights_Tint_Left {
     id=#16
     possessed_class=Image
     tracks [{
       type=MovieSceneWidgetMaterialTrack
       sections [{
         type=MovieSceneParameterSection
         channels [
          FillRadius {type=MovieSceneScriptingFloatChannel, keys [[0,0.6452],[12,0.3]]}
          FinalAlpha {type=MovieSceneScriptingFloatChannel, keys [[0,0.0125],[12,0]]}
          FillDensity {type=MovieSceneScriptingFloatChannel, keys [[0,3.1024],[12,2]]}
         ]
        }]
      }]
    }
    Sights_Container_Right {
     id=#17
     possessed_class=Image
     tracks [
      {
       type=MovieScene2DTransformTrack
       sections [{
         type=MovieScene2DTransformSection
         channels [
          Translation.X {type=MovieSceneScriptingFloatChannel, keys [[0,0],[12,48]]}
          Scale.X {type=MovieSceneScriptingFloatChannel, keys [[0,-1],[12,-1.125]]}
          Scale.Y {type=MovieSceneScriptingFloatChannel, keys [[0,1],[12,1.125]]}
         ]
        }]
      }
      {
       type=MovieSceneWidgetMaterialTrack
       sections [{
         type=MovieSceneParameterSection
         channels [
          FinalAlpha {type=MovieSceneScriptingFloatChannel, keys [[0,0.005],[12,0]]}
          Angle_1_Softness {type=MovieSceneScriptingFloatChannel, keys [[0,32],[12,16]]}
          Bottom_Cut_Softness {type=MovieSceneScriptingFloatChannel, keys [[0,40],[12,20]]}
          Bottom_Softness {type=MovieSceneScriptingFloatChannel, keys [[0,48],[12,24]]}
          Top_Cut_1_Softness {type=MovieSceneScriptingFloatChannel, keys [[0,40],[12,20]]}
          Top_Cut_2_Softness {type=MovieSceneScriptingFloatChannel, keys [[0,40],[12,20]]}
          Top_Softness {type=MovieSceneScriptingFloatChannel, keys [[0,48],[12,24]]}
          Angle_1_Width {type=MovieSceneScriptingFloatChannel, keys [[0,0.1405],[12,0.111]]}
          BottomOffset_Width {type=MovieSceneScriptingFloatChannel, keys [[0,-0.25],[12,-0.67]]}
          TopOffset_Width {type=MovieSceneScriptingFloatChannel, keys [[0,0.7205],[12,0.6905]]}
         ]
        }]
      }
     ]
    }
    Databit_Left {
     id=#18
     possessed_class=Image
     tracks [{
       type=MovieSceneWidgetMaterialTrack
       sections [{
         type=MovieSceneParameterSection
         channels [
          ToBlur {type=MovieSceneScriptingFloatChannel, keys [[0,1],[12,0]]}
          GlowAmount {type=MovieSceneScriptingFloatChannel, keys [[0,0],[12,1]]}
          FinalAlpha {type=MovieSceneScriptingFloatChannel, keys [[0,1],[12,0]]}
         ]
        }]
      }]
    }
    Sights_Container_Blurred_Left {
     id=#19
     possessed_class=Image
     tracks [
      {
       type=MovieScene2DTransformTrack
       sections [{
         type=MovieScene2DTransformSection
         channels [
          Translation.X {type=MovieSceneScriptingFloatChannel, keys [[0,0],[12,24]]}
          Scale.X {type=MovieSceneScriptingFloatChannel, keys [[0,1],[12,0.9]]}
          Scale.Y {type=MovieSceneScriptingFloatChannel, keys [[0,1],[12,0.9]]}
         ]
        }]
      }
      {
       type=MovieSceneWidgetMaterialTrack
       sections [{type=MovieSceneParameterSection, channels [FinalAlpha {type=MovieSceneScriptingFloatChannel, keys [[0,0.05],[12,0]]}]}]
      }
     ]
    }
    Databit_Right {
     id=#20
     possessed_class=Image
     tracks [{
       type=MovieSceneWidgetMaterialTrack
       sections [{
         type=MovieSceneParameterSection
         channels [
          ToBlur {type=MovieSceneScriptingFloatChannel, keys [[0,1],[12,0]]}
          GlowAmount {type=MovieSceneScriptingFloatChannel, keys [[0,0],[12,1]]}
          FinalAlpha {type=MovieSceneScriptingFloatChannel, keys [[0,1],[12,0]]}
         ]
        }]
      }]
    }
    Sights_Tint_Right {
     id=#21
     possessed_class=Image
     tracks [{
       type=MovieSceneWidgetMaterialTrack
       sections [{
         type=MovieSceneParameterSection
         channels [
          FillRadius {type=MovieSceneScriptingFloatChannel, keys [[0,0.6452],[12,0.3]]}
          FinalAlpha {type=MovieSceneScriptingFloatChannel, keys [[0,0.0125],[12,0]]}
          FillDensity {type=MovieSceneScriptingFloatChannel, keys [[0,3.1024],[12,2]]}
         ]
        }]
      }]
    }
    Sights_Container_Blurred_Right {
     id=#22
     possessed_class=Image
     tracks [
      {
       type=MovieScene2DTransformTrack
       sections [{
         type=MovieScene2DTransformSection
         channels [
          Translation.X {type=MovieSceneScriptingFloatChannel, keys [[0,0],[12,-24]]}
          Scale.X {type=MovieSceneScriptingFloatChannel, keys [[0,-1],[12,-0.9]]}
          Scale.Y {type=MovieSceneScriptingFloatChannel, keys [[0,1],[12,0.9]]}
         ]
        }]
      }
      {
       type=MovieSceneWidgetMaterialTrack
       sections [{type=MovieSceneParameterSection, channels [FinalAlpha {type=MovieSceneScriptingFloatChannel, keys [[0,0.05],[12,0]]}]}]
      }
     ]
    }
   ]
   display_rate {numerator=60}
  }
 }
 DepthShapesIntro {
  binding_count=8
  length_seconds=0.5
  object=&self:DepthShapesIntro
  sequence {
   bindings [
    Sights_Container_Left {
     id=#15
     possessed_class=Image
     tracks [
      {
       type=MovieScene2DTransformTrack
       sections [{
         type=MovieScene2DTransformSection
         channels [
          Translation.X {type=MovieSceneScriptingFloatChannel, keys [[0,-48],[6,0]]}
          Scale.X {type=MovieSceneScriptingFloatChannel, keys [[0,1.125],[6,1]]}
          Scale.Y {type=MovieSceneScriptingFloatChannel, keys [[0,1.125],[6,1]]}
         ]
        }]
      }
      {
       type=MovieSceneWidgetMaterialTrack
       sections [{type=MovieSceneParameterSection, channels [FinalAlpha {type=MovieSceneScriptingFloatChannel, keys [[0,0],[18,0.005]]}]}]
      }
     ]
    }
    Sights_Tint_Left {
     id=#16
     possessed_class=Image
     tracks [{
       type=MovieSceneWidgetMaterialTrack
       sections [{
         type=MovieSceneParameterSection
         channels [
          FillRadius {type=MovieSceneScriptingFloatChannel, keys [[0,0.3],[18.002,0.6452]]}
          FinalAlpha {type=MovieSceneScriptingFloatChannel, keys [[0,0],[18.002,0.0125]]}
          FillDensity {type=MovieSceneScriptingFloatChannel, keys [[0,2],[18.002,3.1024]]}
         ]
        }]
      }]
    }
    Sights_Container_Right {
     id=#17
     possessed_class=Image
     tracks [
      {
       type=MovieScene2DTransformTrack
       sections [{
         type=MovieScene2DTransformSection
         channels [
          Translation.X {type=MovieSceneScriptingFloatChannel, keys [[0,48],[6,0]]}
          Scale.X {type=MovieSceneScriptingFloatChannel, keys [[0,-1.125],[6,-1]]}
          Scale.Y {type=MovieSceneScriptingFloatChannel, keys [[0,1.125],[6,1]]}
         ]
        }]
      }
      {
       type=MovieSceneWidgetMaterialTrack
       sections [{type=MovieSceneParameterSection, channels [FinalAlpha {type=MovieSceneScriptingFloatChannel, keys [[0,0],[18,0.005]]}]}]
      }
     ]
    }
    Databit_Left {
     id=#23
     possessed_class=Image
     tracks [{
       type=MovieSceneWidgetMaterialTrack
       sections [{
         type=MovieSceneParameterSection
         channels [
          ToBlur {type=MovieSceneScriptingFloatChannel, keys [[0,0],[30,1]]}
          GlowAmount {type=MovieSceneScriptingFloatChannel, keys [[0,1],[30,0]]}
          FinalAlpha {type=MovieSceneScriptingFloatChannel, keys [[0,0],[6,1],[12,1]]}
         ]
        }]
      }]
    }
    Sights_Container_Blurred_Left {
     id=#19
     possessed_class=Image
     tracks [
      {
       type=MovieScene2DTransformTrack
       sections [{
         type=MovieScene2DTransformSection
         channels [
          Translation.X {type=MovieSceneScriptingFloatChannel, keys [[0,24],[18,0]]}
          Scale.X {type=MovieSceneScriptingFloatChannel, keys [[0,0.9],[18,1]]}
          Scale.Y {type=MovieSceneScriptingFloatChannel, keys [[0,0.9],[18,1]]}
         ]
        }]
      }
      {
       type=MovieSceneWidgetMaterialTrack
       sections [{type=MovieSceneParameterSection, channels [FinalAlpha {type=MovieSceneScriptingFloatChannel, keys [[0,0],[18,0.05]]}]}]
      }
     ]
    }
    Databit_Right {
     id=#24
     possessed_class=Image
     tracks [{
       type=MovieSceneWidgetMaterialTrack
       sections [{
         type=MovieSceneParameterSection
         channels [
          ToBlur {type=MovieSceneScriptingFloatChannel, keys [[0,0],[30,1]]}
          GlowAmount {type=MovieSceneScriptingFloatChannel, keys [[0,1],[30,0]]}
          FinalAlpha {type=MovieSceneScriptingFloatChannel, keys [[0,0],[6,1],[12,1]]}
         ]
        }]
      }]
    }
    Sights_Tint_Right {
     id=#21
     possessed_class=Image
     tracks [{
       type=MovieSceneWidgetMaterialTrack
       sections [{
         type=MovieSceneParameterSection
         channels [
          FillRadius {type=MovieSceneScriptingFloatChannel, keys [[0,0.3],[18.002,0.6452]]}
          FinalAlpha {type=MovieSceneScriptingFloatChannel, keys [[0,0],[18.002,0.0125]]}
          FillDensity {type=MovieSceneScriptingFloatChannel, keys [[0,2],[18.002,3.1024]]}
         ]
        }]
      }]
    }
    Sights_Container_Blurred_Right {
     id=#22
     possessed_class=Image
     tracks [
      {
       type=MovieScene2DTransformTrack
       sections [{
         type=MovieScene2DTransformSection
         channels [
          Translation.X {type=MovieSceneScriptingFloatChannel, keys [[0,-24],[18,0]]}
          Scale.X {type=MovieSceneScriptingFloatChannel, keys [[0,-0.9],[18,-1]]}
          Scale.Y {type=MovieSceneScriptingFloatChannel, keys [[0,0.9],[18,1]]}
         ]
        }]
      }
      {
       type=MovieSceneWidgetMaterialTrack
       sections [{type=MovieSceneParameterSection, channels [FinalAlpha {type=MovieSceneScriptingFloatChannel, keys [[0,0],[18,0.05]]}]}]
      }
     ]
    }
   ]
   display_rate {numerator=60}
  }
 }
 DepthShapesIntro_AnimBurst {
  binding_count=2
  length_seconds=0.3
  object=&self:DepthShapesIntro_AnimBurst
  sequence {bindings [
    Sights_Container_Left_AnimBurst {
     id=#25
     possessed_class=Image
     tracks [
      {
       type=MovieScene2DTransformTrack
       sections [{
         type=MovieScene2DTransformSection
         channels [
          Translation.X {type=MovieSceneScriptingFloatChannel, keys [[0,-48],[2,0]]}
          Scale.X {type=MovieSceneScriptingFloatChannel, keys [[0,1.125],[2,1]]}
          Scale.Y {type=MovieSceneScriptingFloatChannel, keys [[0,1.125],[2,1]]}
         ]
        }]
      }
      {
       type=MovieSceneWidgetMaterialTrack
       sections [{
         type=MovieSceneParameterSection
         channels [
          FinalAlpha {type=MovieSceneScriptingFloatChannel, keys [[0,0],[1,0.025],[6,0]]}
          Angle_1_Softness {type=MovieSceneScriptingFloatChannel, keys [[0,16],[6.0003,32]]}
          Bottom_Cut_Softness {type=MovieSceneScriptingFloatChannel, keys [[0,20],[6.0003,40]]}
          Bottom_Softness {type=MovieSceneScriptingFloatChannel, keys [[0,24],[6.0003,48]]}
          Top_Cut_1_Softness {type=MovieSceneScriptingFloatChannel, keys [[0,20],[6.0003,40]]}
          Top_Cut_2_Softness {type=MovieSceneScriptingFloatChannel, keys [[0,20],[6.0003,40]]}
          Top_Softness {type=MovieSceneScriptingFloatChannel, keys [[0,24],[6.0003,48]]}
          Angle_1_Width {type=MovieSceneScriptingFloatChannel, keys [[0,0.111],[6.0003,0.1405]]}
          BottomOffset_Width {type=MovieSceneScriptingFloatChannel, keys [[0,-0.67],[6.0003,-0.25]]}
          TopOffset_Width {type=MovieSceneScriptingFloatChannel, keys [[0,0.6905],[6.0003,0.7205]]}
         ]
        }]
      }
     ]
    }
    Sights_Container_Right_AnimBurst {
     id=#26
     possessed_class=Image
     tracks [
      {
       type=MovieSceneWidgetMaterialTrack
       sections [{
         type=MovieSceneParameterSection
         channels [
          FinalAlpha {type=MovieSceneScriptingFloatChannel, keys [[0,0],[1,0.025],[6,0]]}
          Angle_1_Softness {type=MovieSceneScriptingFloatChannel, keys [[0,16],[6.0003,32]]}
          Bottom_Cut_Softness {type=MovieSceneScriptingFloatChannel, keys [[0,20],[6.0003,40]]}
          Bottom_Softness {type=MovieSceneScriptingFloatChannel, keys [[0,24],[6.0003,48]]}
          Top_Cut_1_Softness {type=MovieSceneScriptingFloatChannel, keys [[0,20],[6.0003,40]]}
          Top_Cut_2_Softness {type=MovieSceneScriptingFloatChannel, keys [[0,20],[6.0003,40]]}
          Top_Softness {type=MovieSceneScriptingFloatChannel, keys [[0,24],[6.0003,48]]}
          Angle_1_Width {type=MovieSceneScriptingFloatChannel, keys [[0,0.111],[6.0003,0.1405]]}
          BottomOffset_Width {type=MovieSceneScriptingFloatChannel, keys [[0,-0.67],[6.0003,-0.25]]}
          TopOffset_Width {type=MovieSceneScriptingFloatChannel, keys [[0,0.6905],[6.0003,0.7205]]}
         ]
        }]
      }
      {
       type=MovieScene2DTransformTrack
       sections [{
         type=MovieScene2DTransformSection
         channels [
          Translation.X {type=MovieSceneScriptingFloatChannel, keys [[0,48],[2,0]]}
          Scale.X {type=MovieSceneScriptingFloatChannel, keys [[0,-1.125],[2,-1]]}
          Scale.Y {type=MovieSceneScriptingFloatChannel, keys [[0,1.125],[2,1]]}
         ]
        }]
      }
     ]
    }
   ]}
 }
 MinimumChargeReached {
  binding_count=4
  length_seconds=0.5
  object=&self:MinimumChargeReached
  sequence {bindings [
    CenterRing {
     id=#27
     possessed_class=Image
     tracks [
      {
       type=MovieSceneWidgetMaterialTrack
       sections [{type=MovieSceneParameterSection, channels [Overall_Alpha {type=MovieSceneScriptingFloatChannel, keys [[0,0.05],[4,0.1]]}]}]
      }
      {
       type=MovieScene2DTransformTrack
       sections [{
         type=MovieScene2DTransformSection
         channels [
          Translation.X {type=MovieSceneScriptingFloatChannel, const [0,1]}
          Translation.Y {type=MovieSceneScriptingFloatChannel, const [0,1]}
          Angle {type=MovieSceneScriptingFloatChannel, const [0,1]}
          Scale.X {type=MovieSceneScriptingFloatChannel, keys [[0,1],[2,0.75]]}
          Scale.Y {type=MovieSceneScriptingFloatChannel, keys [[0,1],[2,0.75]]}
          Shear.X {type=MovieSceneScriptingFloatChannel, const [0,1]}
          Shear.Y {type=MovieSceneScriptingFloatChannel, const [0,1]}
         ]
        }]
      }
     ]
    }
    Arrow {
     id=#28
     possessed_class=Image
     tracks [
      {
       type=MovieSceneWidgetMaterialTrack
       sections [{type=MovieSceneParameterSection, channels [FinalAlpha {type=MovieSceneScriptingFloatChannel, keys [[0,0.75],[2,1]]}]}]
      }
      {
       type=MovieScene2DTransformTrack
       sections [{
         type=MovieScene2DTransformSection
         channels [
          Translation.X {type=MovieSceneScriptingFloatChannel, const [0,1]}
          Translation.Y {type=MovieSceneScriptingFloatChannel, keys [[0,-70],[2,-64]]}
          Angle {type=MovieSceneScriptingFloatChannel, const [-90,1]}
          Scale.X {type=MovieSceneScriptingFloatChannel, const [1,2]}
          Scale.Y {type=MovieSceneScriptingFloatChannel, const [1,2]}
          Shear.X {type=MovieSceneScriptingFloatChannel, const [0,1]}
          Shear.Y {type=MovieSceneScriptingFloatChannel, const [0,1]}
         ]
        }]
      }
     ]
    }
    Charge_Progress {
     id=#29
     possessed_class=Image
     tracks [{
       type=MovieSceneWidgetMaterialTrack
       sections [{type=MovieSceneParameterSection, channels [Overall_Alpha {type=MovieSceneScriptingFloatChannel, keys [[0,0.175],[1,0.325]]}]}]
      }]
    }
    Arrow_Burst {
     id=#30
     possessed_class=Image
     tracks [
      {
       type=MovieSceneFloatTrack
       display_name=Render Opacity
       sections [{type=MovieSceneFloatSection, channels [None {type=MovieSceneScriptingFloatChannel, keys [[2,0],[3,0.5],[10,0]]}]}]
      }
      {
       type=MovieScene2DTransformTrack
       sections [{
         type=MovieScene2DTransformSection
         channels [
          Translation.X {type=MovieSceneScriptingFloatChannel, const [0,1]}
          Translation.Y {type=MovieSceneScriptingFloatChannel, const [-64,1]}
          Angle {type=MovieSceneScriptingFloatChannel, const [-90,1]}
          Scale.X {type=MovieSceneScriptingFloatChannel, keys [[2,1],[10,3]]}
          Scale.Y {type=MovieSceneScriptingFloatChannel, keys [[2,1],[10,3]]}
          Shear.X {type=MovieSceneScriptingFloatChannel, const [0,1]}
          Shear.Y {type=MovieSceneScriptingFloatChannel, const [0,1]}
         ]
        }]
      }
     ]
    }
   ]}
 }
]
graphs [
 EventGraph {nodes [
   Event Construct()
   > ResetAnim()
   > UpdateChargeToMin!(execute, self, InPercent:double=0) -> (then)
   > UpdateChargeToMax!(execute, self, InPercent:double=0) -> (then)
   n1 = Event OnTargetingChanged!() -> (OutputDelegate, then, bIsADS:bool) {is=title K2Node_Event}
   > n2 = Branch(Condition=@n1.bIsADS)
   > PlayAnimationForward(InAnimation=Get ReticleIntro() -> (ReticleIntro:Widget Animation Object Reference))
   > PlayAnimationForward(InAnimation=Get DepthShapesIntro() -> (DepthShapesIntro:Widget Animation Object Reference))
   > PlayAnimationForward(InAnimation=Get DepthShapesIntro_AnimBurst() -> (DepthShapesIntro_AnimBurst:Widget Animation Object Reference))
   > UpdateChargeToMin(InPercent=0)
   > UpdateChargeToMax(InPercent=0)
   > ResetMinChargeAnimState()
   PlayAnimationReverse(execute=@n2.else, InAnimation=Get ReticleIntro())
   > PlayAnimationReverse(InAnimation=Get DepthShapesIntro())
   > StopAnimation(InAnimation=Get DepthShapesIntro_AnimBurst())
   > ResetMinChargeAnimState()
   n3 = Event OnUpdateChargeToMin!() -> (OutputDelegate, then, ChargePercent:float) {is=title K2Node_Event}
   > Branch(Condition=GetIsADS!(self:Fort Weapon Reticle Extension Widget Base Object Reference) -> (ReturnValue:bool))
   > UpdateChargeToMin(InPercent=@n3.ChargePercent)
   Event OnWeaponSingleFire() {is=title K2Node_Event}
   > PlayAnimation(InAnimation=Get Fire() -> (Fire:Widget Animation Object Reference))
   > UpdateChargeToMin(InPercent=0)
   > UpdateChargeToMax(InPercent=0)
   Event OnMinimumChargeReached() {is=title K2Node_Event}
   > Branch(Condition=GetIsADS())
   > PlayAnimation(InAnimation=Get MinimumChargeReached() -> (MinimumChargeReached:Widget Animation Object Reference))
   n4 = Event OnUpdateChargeToAutoDischarge!() -> (OutputDelegate, then, ChargePercent:float) {is=title K2Node_Event}
   > Branch(Condition=GetIsADS())
   > UpdateChargeToMax(InPercent=@n4.ChargePercent)
   Event Tick()
  ]}
 ResetAnim {nodes [
   ResetAnim() {is=title K2Node_FunctionEntry}
   n1 = Get ReticleIntro()
   > PlayAnimation(InAnimation=@n1, StartAtTime=GetEndTime(self=@n1), PlayMode=Reverse, PlaybackSpeed=16)
   n2 = Get DepthShapesIntro()
   > PlayAnimation(InAnimation=@n2, StartAtTime=GetEndTime(self=@n2), PlayMode=Reverse, PlaybackSpeed=16)
   n3 = Get DepthShapesIntro_AnimBurst()
   > PlayAnimation(InAnimation=@n3, StartAtTime=GetEndTime(self=@n3), PlayMode=Reverse, PlaybackSpeed=16)
   n4 = Get MinimumChargeReached()
   > PlayAnimation(InAnimation=@n4, StartAtTime=GetEndTime(self=@n4), PlayMode=Reverse, PlaybackSpeed=16)
   n5 = Get Fire()
   > PlayAnimation(InAnimation=@n5, StartAtTime=GetEndTime(self=@n5), PlayMode=Reverse, PlaybackSpeed=16)
  ]}
 ResetMinChargeAnimState {nodes [
   ResetMinChargeAnimState() {is=title K2Node_FunctionEntry}
   n1 = Get MinimumChargeReached()
   > PlayAnimation(InAnimation=@n1, StartAtTime=GetEndTime(self=@n1), PlayMode=Reverse)
  ]}
 UpdateChargeToMax {nodes [
   n1 = UpdateChargeToMax!() -> (then, InPercent:double=~) {is=title K2Node_FunctionEntry}
   > n2 = Sequence()
   > n3 = GetDynamicMaterial(self=Get TimerPips_Populated() -> (TimerPips_Populated:Image Object Reference))
   > Is Valid(InputObject=@n3)
   n4 = int * float!(A:int=100, B:double=@n1) -> (ReturnValue:double)
   > SetScalarParameterValue(self=@n3, ParameterName=FillPercent, Value=@n4)
   n5 = GetDynamicMaterial(execute=@n2.then_1, self=Get TimerPips_Glow() -> (TimerPips_Glow:Image Object Reference))
   > Is Valid(InputObject=@n5)
   > SetScalarParameterValue(self=@n5, ParameterName=FillPercent, Value=@n4)
  ]}
 UpdateChargeToMin {nodes [
   n1 = UpdateChargeToMin!() -> (then, InPercent:double=~) {is=title K2Node_FunctionEntry}
   > n2 = GetDynamicMaterial(self=Get Charge_Progress() -> (Charge_Progress:Image Object Reference))
   > Is Valid(InputObject=@n2)
   > SetScalarParameterValue(self=@n2, ParameterName=FillPercent, Value=int * float(A=100, B=@n1))
  ]}
]
variables [
 ADSAnimHasPlayedForward {type=boolean}
 Is ADS {type=boolean}
 ReticleAnimFinished {type=boolean}
]
widgets {
 info {
  inheritedWidgetCount=1
  parentClass=/Script/MotherGameplayUI.FortReactorGradeReticleExtensionWidget
  rootWidgetClass=&Overlay
  widgetCount=34
 }
 widgets [
  Overlay_0 {widgetClassPath=&Overlay, widget=Overlay_0}
  TestBacking_DESIGNVIEWONLY {
   widgetClassPath=&Image
   bIsVariable=true
   parent=Overlay_0
   widget=TestBacking_DESIGNVIEWONLY
   properties {
    visibility=Hidden
    brush {
     bIsDynamicallyLoaded=false
     drawAs=Image
     imageType=NoImage
     mirroring=NoMirror
     resourceName=None
     resourceObject=null
     tiling=NoTile
     imageSize=<1024 1024>
     margin=<ltrb 0 0 0 0>
     outlineSettings {
      bUseBrushTransparency=false
      roundingType=HalfHeightRadius
      width=0
      color {colorUseRule=UseColor_Specified, specifiedColor=<rgba 0 0 0 0>}
      cornerRadii=<0 0 0 0>
     }
     tintColor {colorUseRule=UseColor_Specified, specifiedColor=<rgba 1 1 1 1>}
     uVRegion {bIsValid=false, max=<0 0>, min=<0 0>}
    }
    colorAndOpacity=<rgba 0.0125 0.0125 0.0125 1>
   }
   slot {class=OverlaySlot, properties {horizontalAlignment=HAlign_Fill, verticalAlignment=VAlign_Fill}}
  }
  Overlay_19 {
   widgetClassPath=&Overlay
   parent=Overlay_0
   widget=Overlay_19
   slot {class=OverlaySlot, properties {horizontalAlignment=HAlign_Center, verticalAlignment=VAlign_Center}}
  }
  Sights_Blurred {
   widgetClassPath=&HorizontalBox
   parent=Overlay_19
   widget=Sights_Blurred
   properties {visibility=Hidden, renderTransform {angle=0, scale=<1 1>, shear=<0 0>, translation=<0 -8>}}
   slot {class=OverlaySlot, properties {horizontalAlignment=HAlign_Center, verticalAlignment=VAlign_Center}}
  }
  Sights_Container_Blurred_Left {
   widgetClassPath=&Image
   bIsVariable=true
   parent=Sights_Blurred
   widget=Sights_Container_Blurred_Left
   properties {
    visibility=SelfHitTestInvisible
    brush {
     bIsDynamicallyLoaded=false
     drawAs=Image
     imageType=NoImage
     mirroring=NoMirror
     resourceName=None
     resourceObject=/MotherGameplay/Items/ReactorGrade/UI/Art/M_UI_Reactor_Sights_Component_Blurred
     tiling=NoTile
     imageSize=<640 640>
     margin=<ltrb 0 0 0 0>
     outlineSettings {
      bUseBrushTransparency=false
      roundingType=HalfHeightRadius
      width=0
      color {colorUseRule=UseColor_Specified, specifiedColor=<rgba 0 0 0 0>}
      cornerRadii=<0 0 0 0>
     }
     tintColor {colorUseRule=UseColor_Specified, specifiedColor=<rgba 1 1 1 1>}
     uVRegion {bIsValid=false, max=<0 0>, min=<0 0>}
    }
   }
   slot {class=HorizontalBoxSlot, properties {padding=<ltrb 0 0 -160 0>}}
  }
  Sights_Container_Blurred_Right {
   widgetClassPath=&Image
   bIsVariable=true
   parent=Sights_Blurred
   widget=Sights_Container_Blurred_Right
   properties {
    visibility=SelfHitTestInvisible
    brush {
     bIsDynamicallyLoaded=false
     drawAs=Image
     imageType=NoImage
     mirroring=NoMirror
     resourceName=None
     resourceObject=&M_UI_Reactor_Sights_Component_Blurred
     tiling=NoTile
     imageSize=<640 640>
     margin=<ltrb 0 0 0 0>
     outlineSettings {
      bUseBrushTransparency=false
      roundingType=HalfHeightRadius
      width=0
      color {colorUseRule=UseColor_Specified, specifiedColor=<rgba 0 0 0 0>}
      cornerRadii=<0 0 0 0>
     }
     tintColor {colorUseRule=UseColor_Specified, specifiedColor=<rgba 1 1 1 1>}
     uVRegion {bIsValid=false, max=<0 0>, min=<0 0>}
    }
    renderTransform {angle=0, scale=<-1 1>, shear=<0 0>, translation=<0 0>}
   }
   slot {class=HorizontalBoxSlot, properties {padding=<ltrb -160 0 0 0>}}
  }
  Sights_Tint {
   widgetClassPath=&HorizontalBox
   parent=Overlay_19
   widget=Sights_Tint
   properties {renderTransform {angle=0, scale=<1 1>, shear=<0 0>, translation=<0 -8>}}
   slot {class=OverlaySlot, properties {horizontalAlignment=HAlign_Center, verticalAlignment=VAlign_Center}}
  }
  Sights_Tint_Left {
   widgetClassPath=&Image
   bIsVariable=true
   parent=Sights_Tint
   widget=Sights_Tint_Left
   properties {
    visibility=SelfHitTestInvisible
    brush {
     bIsDynamicallyLoaded=false
     drawAs=Image
     imageType=NoImage
     mirroring=NoMirror
     resourceName=None
     resourceObject=/MotherGameplay/Items/ReactorGrade/UI/Art/M_UI_Reactor_Sights_Component_Tint
     tiling=NoTile
     imageSize=<420 420>
     margin=<ltrb 0 0 0 0>
     outlineSettings {
      bUseBrushTransparency=false
      roundingType=HalfHeightRadius
      width=0
      color {colorUseRule=UseColor_Specified, specifiedColor=<rgba 0 0 0 0>}
      cornerRadii=<0 0 0 0>
     }
     tintColor {colorUseRule=UseColor_Specified, specifiedColor=<rgba 1 1 1 1>}
     uVRegion {bIsValid=false, max=<0 0>, min=<0 0>}
    }
   }
   slot {class=HorizontalBoxSlot, properties {padding=<ltrb 0 0 -96 0>}}
  }
  Sights_Tint_Right {
   widgetClassPath=&Image
   bIsVariable=true
   parent=Sights_Tint
   widget=Sights_Tint_Right
   properties {
    visibility=SelfHitTestInvisible
    brush {
     bIsDynamicallyLoaded=false
     drawAs=Image
     imageType=NoImage
     mirroring=NoMirror
     resourceName=None
     resourceObject=&M_UI_Reactor_Sights_Component_Tint
     tiling=NoTile
     imageSize=<420 420>
     margin=<ltrb 0 0 0 0>
     outlineSettings {
      bUseBrushTransparency=false
      roundingType=HalfHeightRadius
      width=0
      color {colorUseRule=UseColor_Specified, specifiedColor=<rgba 0 0 0 0>}
      cornerRadii=<0 0 0 0>
     }
     tintColor {colorUseRule=UseColor_Specified, specifiedColor=<rgba 1 1 1 1>}
     uVRegion {bIsValid=false, max=<0 0>, min=<0 0>}
    }
    renderTransform {angle=0, scale=<-1 1>, shear=<0 0>, translation=<0 0>}
   }
   slot {class=HorizontalBoxSlot, properties {padding=<ltrb -96 0 0 0>}}
  }
  Sights {
   widgetClassPath=&HorizontalBox
   parent=Overlay_19
   widget=Sights
   properties {renderTransform {angle=0, scale=<1 1>, shear=<0 0>, translation=<0 -8>}}
   slot {class=OverlaySlot, properties {horizontalAlignment=HAlign_Center, verticalAlignment=VAlign_Center}}
  }
  Sights_Container_Left {
   widgetClassPath=&Image
   bIsVariable=true
   parent=Sights
   widget=Sights_Container_Left
   properties {
    visibility=SelfHitTestInvisible
    brush {
     bIsDynamicallyLoaded=false
     drawAs=Image
     imageType=NoImage
     mirroring=NoMirror
     resourceName=None
     resourceObject=/MotherGameplay/Items/ReactorGrade/UI/Art/M_UI_Reactor_Sights_Component_TightBlur
     tiling=NoTile
     imageSize=<420 420>
     margin=<ltrb 0 0 0 0>
     outlineSettings {
      bUseBrushTransparency=false
      roundingType=HalfHeightRadius
      width=0
      color {colorUseRule=UseColor_Specified, specifiedColor=<rgba 0 0 0 0>}
      cornerRadii=<0 0 0 0>
     }
     tintColor {colorUseRule=UseColor_Specified, specifiedColor=<rgba 1 1 1 1>}
     uVRegion {bIsValid=false, max=<0 0>, min=<0 0>}
    }
   }
   slot {class=HorizontalBoxSlot, properties {horizontalAlignment=HAlign_Left, verticalAlignment=VAlign_Center, padding=<ltrb 0 0 -96 0>}}
  }
  Sights_Container_Right {
   widgetClassPath=&Image
   bIsVariable=true
   parent=Sights
   widget=Sights_Container_Right
   properties {
    visibility=SelfHitTestInvisible
    brush {
     bIsDynamicallyLoaded=false
     drawAs=Image
     imageType=NoImage
     mirroring=NoMirror
     resourceName=None
     resourceObject=&M_UI_Reactor_Sights_Component_TightBlur
     tiling=NoTile
     imageSize=<420 420>
     margin=<ltrb 0 0 0 0>
     outlineSettings {
      bUseBrushTransparency=false
      roundingType=HalfHeightRadius
      width=0
      color {colorUseRule=UseColor_Specified, specifiedColor=<rgba 0 0 0 0>}
      cornerRadii=<0 0 0 0>
     }
     tintColor {colorUseRule=UseColor_Specified, specifiedColor=<rgba 1 1 1 1>}
     uVRegion {bIsValid=false, max=<0 0>, min=<0 0>}
    }
    renderTransform {angle=0, scale=<-1 1>, shear=<0 0>, translation=<0 0>}
   }
   slot {class=HorizontalBoxSlot, properties {horizontalAlignment=HAlign_Right, padding=<ltrb -96 0 0 0>}}
  }
  Sights_Burst {
   widgetClassPath=&HorizontalBox
   parent=Overlay_19
   widget=Sights_Burst
   properties {renderTransform {angle=0, scale=<1 1>, shear=<0 0>, translation=<0 -8>}}
   slot {class=OverlaySlot, properties {horizontalAlignment=HAlign_Center, verticalAlignment=VAlign_Center}}
  }
  Sights_Container_Left_AnimBurst {
   widgetClassPath=&Image
   bIsVariable=true
   parent=Sights_Burst
   widget=Sights_Container_Left_AnimBurst
   properties {
    visibility=SelfHitTestInvisible
    brush {
     bIsDynamicallyLoaded=false
     drawAs=Image
     imageType=NoImage
     mirroring=NoMirror
     resourceName=None
     resourceObject=&M_UI_Reactor_Sights_Component_TightBlur
     tiling=NoTile
     imageSize=<420 420>
     margin=<ltrb 0 0 0 0>
     outlineSettings {
      bUseBrushTransparency=false
      roundingType=HalfHeightRadius
      width=0
      color {colorUseRule=UseColor_Specified, specifiedColor=<rgba 0 0 0 0>}
      cornerRadii=<0 0 0 0>
     }
     tintColor {colorUseRule=UseColor_Specified, specifiedColor=<rgba 1 1 1 1>}
     uVRegion {bIsValid=false, max=<0 0>, min=<0 0>}
    }
   }
   slot {class=HorizontalBoxSlot, properties {horizontalAlignment=HAlign_Left, verticalAlignment=VAlign_Center, padding=<ltrb 0 0 -96 0>}}
  }
  Sights_Container_Right_AnimBurst {
   widgetClassPath=&Image
   bIsVariable=true
   parent=Sights_Burst
   widget=Sights_Container_Right_AnimBurst
   properties {
    visibility=SelfHitTestInvisible
    brush {
     bIsDynamicallyLoaded=false
     drawAs=Image
     imageType=NoImage
     mirroring=NoMirror
     resourceName=None
     resourceObject=&M_UI_Reactor_Sights_Component_TightBlur
     tiling=NoTile
     imageSize=<420 420>
     margin=<ltrb 0 0 0 0>
     outlineSettings {
      bUseBrushTransparency=false
      roundingType=HalfHeightRadius
      width=0
      color {colorUseRule=UseColor_Specified, specifiedColor=<rgba 0 0 0 0>}
      cornerRadii=<0 0 0 0>
     }
     tintColor {colorUseRule=UseColor_Specified, specifiedColor=<rgba 1 1 1 1>}
     uVRegion {bIsValid=false, max=<0 0>, min=<0 0>}
    }
    renderTransform {angle=0, scale=<-1 1>, shear=<0 0>, translation=<0 0>}
   }
   slot {class=HorizontalBoxSlot, properties {horizontalAlignment=HAlign_Right, padding=<ltrb -96 0 0 0>}}
  }
  EncompassingRing_Shadow {
   widgetClassPath=&Image
   bIsVariable=true
   parent=Overlay_19
   widget=EncompassingRing_Shadow
   properties {brush {
     bIsDynamicallyLoaded=false
     drawAs=Image
     imageType=NoImage
     mirroring=NoMirror
     resourceName=None
     resourceObject=/MotherGameplay/Items/ReactorGrade/UI/Art/M_UI_Reactor_Dark_Indicator_Ring_Shadow
     tiling=NoTile
     imageSize=<160 160>
     margin=<ltrb 0 0 0 0>
     outlineSettings {
      bUseBrushTransparency=false
      roundingType=HalfHeightRadius
      width=0
      color {colorUseRule=UseColor_Specified, specifiedColor=<rgba 0 0 0 0>}
      cornerRadii=<0 0 0 0>
     }
     tintColor {colorUseRule=UseColor_Specified, specifiedColor=<rgba 1 1 1 1>}
     uVRegion {bIsValid=false, max=<0 0>, min=<0 0>}
    }}
   slot {class=OverlaySlot, properties {horizontalAlignment=HAlign_Center, verticalAlignment=VAlign_Center}}
  }
  EncompassingRing {
   widgetClassPath=&Image
   bIsVariable=true
   parent=Overlay_19
   widget=EncompassingRing
   properties {
    visibility=SelfHitTestInvisible
    brush {
     bIsDynamicallyLoaded=false
     drawAs=Image
     imageType=NoImage
     mirroring=NoMirror
     resourceName=None
     resourceObject=/MotherGameplay/Items/ReactorGrade/UI/Art/M_UI_Reactor_Dark_Indicator_Ring
     tiling=NoTile
     imageSize=<160 160>
     margin=<ltrb 0 0 0 0>
     outlineSettings {
      bUseBrushTransparency=false
      roundingType=HalfHeightRadius
      width=0
      color {colorUseRule=UseColor_Specified, specifiedColor=<rgba 0 0 0 0>}
      cornerRadii=<0 0 0 0>
     }
     tintColor {colorUseRule=UseColor_Specified, specifiedColor=<rgba 1 1 1 1>}
     uVRegion {bIsValid=false, max=<0 0>, min=<0 0>}
    }
   }
   slot {class=OverlaySlot, properties {horizontalAlignment=HAlign_Center, verticalAlignment=VAlign_Center}}
  }
  LargeBlueRing {
   widgetClassPath=&Image
   bIsVariable=true
   parent=Overlay_19
   widget=LargeBlueRing
   properties {
    visibility=SelfHitTestInvisible
    brush {
     bIsDynamicallyLoaded=false
     drawAs=Image
     imageType=NoImage
     mirroring=NoMirror
     resourceName=None
     resourceObject=/MotherGameplay/Items/ReactorGrade/UI/Art/M_UI_Reactor_Circle_FullRing
     tiling=NoTile
     imageSize=<128 128>
     margin=<ltrb 0 0 0 0>
     outlineSettings {
      bUseBrushTransparency=false
      roundingType=HalfHeightRadius
      width=0
      color {colorUseRule=UseColor_Specified, specifiedColor=<rgba 0 0 0 0>}
      cornerRadii=<0 0 0 0>
     }
     tintColor {colorUseRule=UseColor_Specified, specifiedColor=<rgba 1 1 1 1>}
     uVRegion {bIsValid=false, max=<0 0>, min=<0 0>}
    }
   }
   slot {class=OverlaySlot, properties {horizontalAlignment=HAlign_Center, verticalAlignment=VAlign_Center}}
  }
  LargeBlueRing_Burst {
   widgetClassPath=&Image
   bIsVariable=true
   parent=Overlay_19
   widget=LargeBlueRing_Burst
   properties {
    renderOpacity=0
    visibility=SelfHitTestInvisible
    brush {
     bIsDynamicallyLoaded=false
     drawAs=Image
     imageType=NoImage
     mirroring=NoMirror
     resourceName=None
     resourceObject=&M_UI_Reactor_Circle_FullRing
     tiling=NoTile
     imageSize=<128 128>
     margin=<ltrb 0 0 0 0>
     outlineSettings {
      bUseBrushTransparency=false
      roundingType=HalfHeightRadius
      width=0
      color {colorUseRule=UseColor_Specified, specifiedColor=<rgba 0 0 0 0>}
      cornerRadii=<0 0 0 0>
     }
     tintColor {colorUseRule=UseColor_Specified, specifiedColor=<rgba 1 1 1 1>}
     uVRegion {bIsValid=false, max=<0 0>, min=<0 0>}
    }
   }
   slot {class=OverlaySlot, properties {horizontalAlignment=HAlign_Center, verticalAlignment=VAlign_Center}}
  }
  CenterRing {
   widgetClassPath=&Image
   bIsVariable=true
   parent=Overlay_19
   widget=CenterRing
   properties {
    visibility=SelfHitTestInvisible
    brush {
     bIsDynamicallyLoaded=false
     drawAs=Image
     imageType=NoImage
     mirroring=NoMirror
     resourceName=None
     resourceObject=/MotherGameplay/Items/ReactorGrade/UI/Art/M_UI_Reactor_Center_Ring
     tiling=NoTile
     imageSize=<32 32>
     margin=<ltrb 0 0 0 0>
     outlineSettings {
      bUseBrushTransparency=false
      roundingType=HalfHeightRadius
      width=0
      color {colorUseRule=UseColor_Specified, specifiedColor=<rgba 0 0 0 0>}
      cornerRadii=<0 0 0 0>
     }
     tintColor {colorUseRule=UseColor_Specified, specifiedColor=<rgba 1 1 1 1>}
     uVRegion {bIsValid=false, max=<0 0>, min=<0 0>}
    }
   }
   slot {class=OverlaySlot, properties {horizontalAlignment=HAlign_Center, verticalAlignment=VAlign_Center}}
  }
  CenterDot_DropShadow {
   widgetClassPath=&Image
   bIsVariable=true
   parent=Overlay_19
   widget=CenterDot_DropShadow
   properties {
    visibility=Collapsed
    brush {
     bIsDynamicallyLoaded=false
     drawAs=Image
     imageType=NoImage
     mirroring=NoMirror
     resourceName=None
     resourceObject=/MotherGameplay/Items/ReactorGrade/UI/Art/M_UI_Reactor_Dot_DropShadow
     tiling=NoTile
     imageSize=<16 16>
     margin=<ltrb 0 0 0 0>
     outlineSettings {
      bUseBrushTransparency=false
      roundingType=HalfHeightRadius
      width=0
      color {colorUseRule=UseColor_Specified, specifiedColor=<rgba 0 0 0 0>}
      cornerRadii=<0 0 0 0>
     }
     tintColor {colorUseRule=UseColor_Specified, specifiedColor=<rgba 1 1 1 1>}
     uVRegion {bIsValid=false, max=<0 0>, min=<0 0>}
    }
   }
   slot {class=OverlaySlot, properties {horizontalAlignment=HAlign_Center, verticalAlignment=VAlign_Center}}
  }
  CenterDot {
   widgetClassPath=&Image
   bIsVariable=true
   parent=Overlay_19
   widget=CenterDot
   properties {
    visibility=SelfHitTestInvisible
    brush {
     bIsDynamicallyLoaded=false
     drawAs=Image
     imageType=NoImage
     mirroring=NoMirror
     resourceName=None
     resourceObject=/MotherGameplay/Items/ReactorGrade/UI/Art/M_UI_Reactor_Dot
     tiling=NoTile
     imageSize=<4 4>
     margin=<ltrb 0 0 0 0>
     outlineSettings {
      bUseBrushTransparency=false
      roundingType=HalfHeightRadius
      width=0
      color {colorUseRule=UseColor_Specified, specifiedColor=<rgba 0 0 0 0>}
      cornerRadii=<0 0 0 0>
     }
     tintColor {colorUseRule=UseColor_Specified, specifiedColor=<rgba 1 1 1 1>}
     uVRegion {bIsValid=false, max=<0 0>, min=<0 0>}
    }
   }
   slot {class=OverlaySlot, properties {horizontalAlignment=HAlign_Center, verticalAlignment=VAlign_Center}}
  }
  TimerPips_Base {
   widgetClassPath=&Image
   bIsVariable=true
   parent=Overlay_19
   widget=TimerPips_Base
   properties {
    visibility=SelfHitTestInvisible
    brush {
     bIsDynamicallyLoaded=false
     drawAs=Image
     imageType=NoImage
     mirroring=NoMirror
     resourceName=None
     resourceObject=/MotherGameplay/Items/ReactorGrade/UI/Art/M_UI_Reactor_Timer_Pips_Base
     tiling=NoTile
     imageSize=<256 256>
     margin=<ltrb 0 0 0 0>
     outlineSettings {
      bUseBrushTransparency=false
      roundingType=HalfHeightRadius
      width=0
      color {colorUseRule=UseColor_Specified, specifiedColor=<rgba 0 0 0 0>}
      cornerRadii=<0 0 0 0>
     }
     tintColor {colorUseRule=UseColor_Specified, specifiedColor=<rgba 1 1 1 1>}
     uVRegion {bIsValid=false, max=<0 0>, min=<0 0>}
    }
   }
   slot {class=OverlaySlot, properties {horizontalAlignment=HAlign_Center, verticalAlignment=VAlign_Center}}
  }
  TimerPips_Glow {
   widgetClassPath=&Image
   bIsVariable=true
   parent=Overlay_19
   widget=TimerPips_Glow
   properties {
    visibility=SelfHitTestInvisible
    brush {
     bIsDynamicallyLoaded=false
     drawAs=Image
     imageType=NoImage
     mirroring=NoMirror
     resourceName=None
     resourceObject=/MotherGameplay/Items/ReactorGrade/UI/Art/M_UI_Reactor_Timer_Pips_Glow
     tiling=NoTile
     imageSize=<256 256>
     margin=<ltrb 0 0 0 0>
     outlineSettings {
      bUseBrushTransparency=false
      roundingType=HalfHeightRadius
      width=0
      color {colorUseRule=UseColor_Specified, specifiedColor=<rgba 0 0 0 0>}
      cornerRadii=<0 0 0 0>
     }
     tintColor {colorUseRule=UseColor_Specified, specifiedColor=<rgba 1 1 1 1>}
     uVRegion {bIsValid=false, max=<0 0>, min=<0 0>}
    }
   }
   slot {class=OverlaySlot, properties {horizontalAlignment=HAlign_Center, verticalAlignment=VAlign_Center}}
  }
  TimerPips_Populated {
   widgetClassPath=&Image
   bIsVariable=true
   parent=Overlay_19
   widget=TimerPips_Populated
   properties {brush {
     bIsDynamicallyLoaded=false
     drawAs=Image
     imageType=NoImage
     mirroring=NoMirror
     resourceName=None
     resourceObject=/MotherGameplay/Items/ReactorGrade/UI/Art/M_UI_Reactor_Timer_Pips
     tiling=NoTile
     imageSize=<256 256>
     margin=<ltrb 0 0 0 0>
     outlineSettings {
      bUseBrushTransparency=false
      roundingType=HalfHeightRadius
      width=0
      color {colorUseRule=UseColor_Specified, specifiedColor=<rgba 0 0 0 0>}
      cornerRadii=<0 0 0 0>
     }
     tintColor {colorUseRule=UseColor_Specified, specifiedColor=<rgba 1 1 1 1>}
     uVRegion {bIsValid=false, max=<0 0>, min=<0 0>}
    }}
   slot {class=OverlaySlot, properties {horizontalAlignment=HAlign_Center, verticalAlignment=VAlign_Center}}
  }
  Charge_Progress_Backing {
   widgetClassPath=&Image
   bIsVariable=true
   parent=Overlay_19
   widget=Charge_Progress_Backing
   properties {
    visibility=SelfHitTestInvisible
    brush {
     bIsDynamicallyLoaded=false
     drawAs=Image
     imageType=NoImage
     mirroring=NoMirror
     resourceName=None
     resourceObject=/MotherGameplay/Items/ReactorGrade/UI/Art/M_UI_Reactor_Charge_Progress_Backing
     tiling=NoTile
     imageSize=<192 192>
     margin=<ltrb 0 0 0 0>
     outlineSettings {
      bUseBrushTransparency=false
      roundingType=HalfHeightRadius
      width=0
      color {colorUseRule=UseColor_Specified, specifiedColor=<rgba 0 0 0 0>}
      cornerRadii=<0 0 0 0>
     }
     tintColor {colorUseRule=UseColor_Specified, specifiedColor=<rgba 1 1 1 1>}
     uVRegion {bIsValid=false, max=<0 0>, min=<0 0>}
    }
   }
   slot {class=OverlaySlot, properties {horizontalAlignment=HAlign_Center, verticalAlignment=VAlign_Center}}
  }
  Charge_Progress {
   widgetClassPath=&Image
   bIsVariable=true
   parent=Overlay_19
   widget=Charge_Progress
   properties {
    visibility=SelfHitTestInvisible
    brush {
     bIsDynamicallyLoaded=false
     drawAs=Image
     imageType=NoImage
     mirroring=NoMirror
     resourceName=None
     resourceObject=/MotherGameplay/Items/ReactorGrade/UI/Art/M_UI_Reactor_Charge_Progress
     tiling=NoTile
     imageSize=<192 192>
     margin=<ltrb 0 0 0 0>
     outlineSettings {
      bUseBrushTransparency=false
      roundingType=HalfHeightRadius
      width=0
      color {colorUseRule=UseColor_Specified, specifiedColor=<rgba 0 0 0 0>}
      cornerRadii=<0 0 0 0>
     }
     tintColor {colorUseRule=UseColor_Specified, specifiedColor=<rgba 1 1 1 1>}
     uVRegion {bIsValid=false, max=<0 0>, min=<0 0>}
    }
   }
   slot {class=OverlaySlot, properties {horizontalAlignment=HAlign_Center, verticalAlignment=VAlign_Center}}
  }
  Charge_Outer_Ring {
   widgetClassPath=&Image
   bIsVariable=true
   parent=Overlay_19
   widget=Charge_Outer_Ring
   properties {
    visibility=SelfHitTestInvisible
    brush {
     bIsDynamicallyLoaded=false
     drawAs=Image
     imageType=NoImage
     mirroring=NoMirror
     resourceName=None
     resourceObject=/MotherGameplay/Items/ReactorGrade/UI/Art/M_UI_Reactor_Light_Progress_Container
     tiling=NoTile
     imageSize=<192 192>
     margin=<ltrb 0 0 0 0>
     outlineSettings {
      bUseBrushTransparency=false
      roundingType=HalfHeightRadius
      width=0
      color {colorUseRule=UseColor_Specified, specifiedColor=<rgba 0 0 0 0>}
      cornerRadii=<0 0 0 0>
     }
     tintColor {colorUseRule=UseColor_Specified, specifiedColor=<rgba 1 1 1 1>}
     uVRegion {bIsValid=false, max=<0 0>, min=<0 0>}
    }
   }
   slot {class=OverlaySlot, properties {horizontalAlignment=HAlign_Center, verticalAlignment=VAlign_Center}}
  }
  VerticalLine {
   widgetClassPath=&Image
   bIsVariable=true
   parent=Overlay_19
   widget=VerticalLine
   properties {
    renderOpacity=0.2
    visibility=SelfHitTestInvisible
    brush {
     bIsDynamicallyLoaded=false
     drawAs=Image
     imageType=NoImage
     mirroring=NoMirror
     resourceName=None
     resourceObject=null
     tiling=NoTile
     imageSize=<2 24>
     margin=<ltrb 0 0 0 0>
     outlineSettings {
      bUseBrushTransparency=false
      roundingType=HalfHeightRadius
      width=0
      color {colorUseRule=UseColor_Specified, specifiedColor=<rgba 0 0 0 0>}
      cornerRadii=<0 0 0 0>
     }
     tintColor {colorUseRule=UseColor_Specified, specifiedColor=<rgba 1 0.5841 0.5029 1>}
     uVRegion {bIsValid=false, max=<0 0>, min=<0 0>}
    }
    renderTransform {angle=0, scale=<1 1>, shear=<0 0>, translation=<0 -82>}
    renderTransformPivot=<0.5 0>
   }
   slot {class=OverlaySlot, properties {horizontalAlignment=HAlign_Center, verticalAlignment=VAlign_Center}}
  }
  Arrow_Burst {
   widgetClassPath=&Image
   bIsVariable=true
   parent=Overlay_19
   widget=Arrow_Burst
   properties {
    renderOpacity=0
    visibility=SelfHitTestInvisible
    brush {
     bIsDynamicallyLoaded=false
     drawAs=Image
     imageType=NoImage
     mirroring=NoMirror
     resourceName=None
     resourceObject=/MotherGameplay/Items/ReactorGrade/UI/Art/M_UI_Reactor_Arrow
     tiling=NoTile
     imageSize=<12 12>
     margin=<ltrb 0 0 0 0>
     outlineSettings {
      bUseBrushTransparency=false
      roundingType=HalfHeightRadius
      width=0
      color {colorUseRule=UseColor_Specified, specifiedColor=<rgba 0 0 0 0>}
      cornerRadii=<0 0 0 0>
     }
     tintColor {colorUseRule=UseColor_Specified, specifiedColor=<rgba 1 1 1 1>}
     uVRegion {bIsValid=false, max=<0 0>, min=<0 0>}
    }
    renderTransform {angle=-90, scale=<1 1>, shear=<0 0>, translation=<0 -64>}
   }
   slot {class=OverlaySlot, properties {horizontalAlignment=HAlign_Center, verticalAlignment=VAlign_Center}}
  }
  Arrow {
   widgetClassPath=&Image
   bIsVariable=true
   parent=Overlay_19
   widget=Arrow
   properties {
    visibility=SelfHitTestInvisible
    brush {
     bIsDynamicallyLoaded=false
     drawAs=Image
     imageType=NoImage
     mirroring=NoMirror
     resourceName=None
     resourceObject=&M_UI_Reactor_Arrow
     tiling=NoTile
     imageSize=<12 12>
     margin=<ltrb 0 0 0 0>
     outlineSettings {
      bUseBrushTransparency=false
      roundingType=HalfHeightRadius
      width=0
      color {colorUseRule=UseColor_Specified, specifiedColor=<rgba 0 0 0 0>}
      cornerRadii=<0 0 0 0>
     }
     tintColor {colorUseRule=UseColor_Specified, specifiedColor=<rgba 1 1 1 1>}
     uVRegion {bIsValid=false, max=<0 0>, min=<0 0>}
    }
    renderTransform {angle=-90, scale=<1 1>, shear=<0 0>, translation=<0 -70>}
   }
   slot {class=OverlaySlot, properties {horizontalAlignment=HAlign_Center, verticalAlignment=VAlign_Center}}
  }
  Databit_Left {
   widgetClassPath=&Image
   bIsVariable=true
   parent=Overlay_19
   widget=Databit_Left
   properties {
    visibility=SelfHitTestInvisible
    brush {
     bIsDynamicallyLoaded=false
     drawAs=Image
     imageType=NoImage
     mirroring=NoMirror
     resourceName=None
     resourceObject=/MotherGameplay/Items/ReactorGrade/UI/Art/M_UI_ReactorGrade_Databit_2
     tiling=NoTile
     imageSize=<88 44>
     margin=<ltrb 0 0 0 0>
     outlineSettings {
      bUseBrushTransparency=false
      roundingType=HalfHeightRadius
      width=0
      color {colorUseRule=UseColor_Specified, specifiedColor=<rgba 0 0 0 0>}
      cornerRadii=<0 0 0 0>
     }
     tintColor {colorUseRule=UseColor_Specified, specifiedColor=<rgba 1 1 1 1>}
     uVRegion {bIsValid=false, max=<0 0>, min=<0 0>}
    }
   }
   slot {class=OverlaySlot, properties {horizontalAlignment=HAlign_Center, verticalAlignment=VAlign_Bottom, padding=<ltrb 0 0 64 128>}}
  }
  Databit_Right {
   widgetClassPath=&Image
   bIsVariable=true
   parent=Overlay_19
   widget=Databit_Right
   properties {
    visibility=SelfHitTestInvisible
    brush {
     bIsDynamicallyLoaded=false
     drawAs=Image
     imageType=NoImage
     mirroring=NoMirror
     resourceName=None
     resourceObject=/MotherGameplay/Items/ReactorGrade/UI/Art/M_UI_ReactorGrade_Databit_1
     tiling=NoTile
     imageSize=<88 44>
     margin=<ltrb 0 0 0 0>
     outlineSettings {
      bUseBrushTransparency=false
      roundingType=HalfHeightRadius
      width=0
      color {colorUseRule=UseColor_Specified, specifiedColor=<rgba 0 0 0 0>}
      cornerRadii=<0 0 0 0>
     }
     tintColor {colorUseRule=UseColor_Specified, specifiedColor=<rgba 1 1 1 1>}
     uVRegion {bIsValid=false, max=<0 0>, min=<0 0>}
    }
   }
   slot {class=OverlaySlot, properties {horizontalAlignment=HAlign_Center, verticalAlignment=VAlign_Bottom, padding=<ltrb 64 0 0 128>}}
  }
  TutorialHighlight {widgetClassPath=&TutorialHighlight}
 ]
}
